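# Translucent, two-sided unlit material for the HighTower Plum KineticAbs energy dome, built for Niagara mesh particles.
# Particle color tints panning plasma clouds and noise-distorted nebula, while noise pushes vertices along normals.
# Opacity combines bottom falloff, fresnel and hard-edged dissolve, with non-mobile depth fade and optional time-of-day scaling.
# Dynamic parameters drive refraction, erosion and displacement; ErodeHardness is unused, and refraction is constant on forward and mobile paths.
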Material /HighTower/Items/Plum/KineticAbs/FX/MI_HightTower_Plum_EnergyDome
graph {
 properties {
  bUsedWithNiagaraMeshParticles=true
  blendMode=BLEND_Translucent
  pixelDepthOffsetMode=PDOM_Legacy
  refractionMethod=RM_IndexOfRefraction
  shadingModel=MSM_Unlit
  twoSided=true
  usedShadingModels=Unlit
 }
 nodes [
  n1 = TextureCoordinate()
  n2 = Sine(None=Multiply(A=MaterialFunctionCall(Float2=@n1) {materialFunction=&BreakOutFloat2Components}.G) {constB=0.5})
  n3 = VertexNormalWS()
  n4 = TextureSample(UVs=Multiply(A=TextureCoordinate()) {constB=4}) {samplerType=SAMPLERTYPE_LinearColor, texture=&HighContrastNoise}
  Multiply(
   A=Saturate(None=Multiply(A=@n2) {constB=2})
   B=Multiply(
    A=Multiply(A=@n3, B=@n4.RGB)
    B=DynamicParameter() {defaultValue=<rgba 1 1 1 100>, paramNames [RefractionIndex, Erode, ErodeHardness, WPO]}.WPO
   )
  )
  n5 = TextureCoordinate()
  n6 = Sine(None=Multiply(A=MaterialFunctionCall(Float2=@n5) {materialFunction=&BreakOutFloat2Components}.G) {constB=0.5})
  n7 = Multiply(
   A=@n6
   B=TextureSampleParameter2D(UVs=Panner() {speedY=-0.2}) {parameterName=Texture_A, texture=/Game/Weapons/WeaponSkins/Wraps/Textures/PlasmaClouds/T_Wrap_PlasmaClouds__M}.G
  )
  n8 = TextureSample(UVs=Panner(Coordinate=TextureCoordinate() {uTiling=5, vTiling=5}) {speedY=-0.2}) {samplerType=SAMPLERTYPE_LinearColor, texture=&HighContrastNoise}
  n9 = Add(A=TextureCoordinate() {uTiling=5, vTiling=5}, B=Divide(A=ComponentMask(None=@n8.RGB) {g=true, r=true}) {constB=6})
  n10 = Multiply(
   A=@n6
   B=Desaturation(
    None=TextureSampleParameter2D(UVs=@n9) {parameterName=Texture_B, texture=/Game/Characters/Player/Male/Medium/Bodies/M_Med_Soldier_04/Skins/Galaxy/Textures/T_FN_Nebula}.RGB
    Fraction=Constant() {r=0.65}
   )
  )
  n11 = ParticleColor()
  Multiply(A=Add(A=Multiply(A=@n7) {constB=0.6}, B=Multiply(A=@n10) {constB=0.65}), B=@n11.RGB)
  n12 = MaterialFunctionCall(
   In=OneMinus(None=MaterialFunctionCall() {materialFunction=&LinearGradient}.VGradient)
   Contrast=ScalarParameter() {defaultValue=0.05, parameterName=BottomFalloff}
  ) {materialFunction=&CheapContrast}
  n13 = MaterialFunctionCall(In=OneMinus(None=Fresnel() {exponent=1}), Contrast=Constant() {r=0.2}) {materialFunction=&CheapContrast}
  n14 = Constant() {r=1}
  ShadingPathSwitch(
   Default=LinearInterpolate(
    B=DynamicParameter() {defaultValue=<rgba 1 1 1 100>, paramNames [RefractionIndex, Erode, ErodeHardness, WPO]}.RefractionIndex
    Alpha=Multiply(A=@n12.Result, B=@n13.Result)
   ) {constA=1}
   Forward=@n14
   Mobile=@n14
  )
  n15 = ScalarParameter() {defaultValue=1, parameterName=OpacityMorning}
  n16 = ScalarParameter() {defaultValue=1, parameterName=OpacityDay}
  n17 = ScalarParameter() {defaultValue=1, parameterName=OpacityEvening}
  n18 = ScalarParameter() {defaultValue=1, parameterName=OpacityNight}
  n19 = Fresnel() {exponent=1}
  n20 = Clamp(None=Multiply(A=Multiply(A=@n19, B=@n13.Result), B=ScalarParameter() {defaultValue=1, parameterName=FresnelBrightness}))
  n21 = DynamicParameter() {defaultValue=<rgba 1 1 1 100>, paramNames [RefractionIndex, Erode, ErodeHardness, WPO]}
  n22 = Add(
   A=TextureSampleParameter2D(UVs=TextureCoordinate() {vTiling=1.5}) {parameterName=Dissolve, texture=/Game/Effects/Fort_Effects/Textures/T_Smooth_tile}.R
   B=Multiply(A=@n21.Erode, B=@n21.Erode)
  )
  n23 = Multiply(A=Multiply(A=@n11.A, B=Multiply(A=@n12.Result, B=@n20)), B=Ceil(None=Saturate(None=@n22)))
  n24 = ShadingPathSwitch(Default=DepthFade(Opacity=@n23, FadeDistance=ScalarParameter() {defaultValue=200, parameterName=FadeDistance}), Mobile=@n23) {desc=Migrated from FeatureLevelSwitch}
  StaticSwitchParameter(
   True=Multiply(
    A=MaterialFunctionCall(Color_Morning=@n15, Color_Day=@n16, Color_Evening=@n17, Color_Night=@n18) {materialFunction=&Color_Time_of_Day}.Result
    B=@n24
   )
   False=@n24
  ) {parameterName=Use_color_time_of_day}
 ]
}
