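; Gameplay ability for the HighTower Plum Kinetic Abs explosion, which damages and knocks back targets after a timed effect.
; Activation binds the player pawn's damage event to an impact gameplay cue and starts the duration timer.
; On expiry it fires the explode cue, runs target selection with a failsafe delay, and applies the damage effect.
; Ending the ability clears the timer and removes the damage-reduction effect by granted tag.
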(event Ability|EventOnEndAbility
  (params (bWasCancelled :type Boolean))
  ;[1 1 1] On end ability clear handle and remove the Damage Reduction GE
  (Utilities|Time|ClearandInvalidateTimerbyHandle (Variables|Default|GetTimer_EffectDuration))
  (Ability|RemoveGameplayEffectFromOwnerWithGrantedTags (Variables|Default|GetT_DRGrantedTag)))

(event Ability|EventActivateAbility
  (bind _returnvalue (Ability|GetAvatarActorfromActorInfo))
  ;[1 1 1] On activate start the timer for the targeting and end of the ability.
  (Utilities|IsValid _returnvalue
    (:Is%20Valid
      (bind _asfort_player_pawn_athena (Utilities|Casting|CastToFortPlayerPawnAthena _returnvalue)
        (:then
          (bind _output_get (Variables|Default|SetPlayerPawn _asfort_player_pawn_athena))
          (EventDispatchers|BindEventtoOnDamaged _output_get (delegate Custom|Event%20On%20Damaged%20))
          (bind _returnvalue_2 (Utilities|Time|SetTimerbyEvent (EventDispatchers|CreateEvent :!event Custom|Explode) (Variables|Default|GetEffectDuration)))
          (Variables|Default|SetTimer_EffectDuration _returnvalue_2))
        (:CastFailed
          (Ability|EndAbility))))
    (:Is%20Not%20Valid
      (Ability|EndAbility))))

(event Custom|Explode
  ;[1 1 1] On explode play the Explode GC and do targeting / apply the damage / knockback. After end the ability.
  (Ability|ExecuteGameplayCueOnOwner (Variables|Default|GetGC_Explode))
  (Networking|ForceNetUpdate (Ability|GetAvatarActorfromActorInfo))
  (bind (_asynctaskproxy _targetdata) (Ability|Tasks|DoTargetSelection "None" (Variables|Default|GetT_TargetingApplication) true)
    (:then
      (Utilities|FlowControl|Delay (Variables|Default|GetTargetingFailsafeDelay))
      (Ability|RemoveGameplayCueFromOwner (Variables|Default|GetGC_Explode)))
    (:Targeted
      (Ability|ApplyGameplayEffectToTarget _targetdata (Variables|Default|GetGE_Damage) 0)
      (Utilities|FlowControl|Delay 0.01) ; Wait a frame after application to end the ability
      (Ability|EndAbility))
    (:Cancelled
      _))
  (Ability|EndAbility))

(event Custom|Event%20On%20Damaged%20
  (params (DamagedActor :type Actor%20Object%20Reference) (Damage :type Float%20%28single-precision%29) (InstigatedBy :type Controller%20Object%20Reference) (DamageCauser :type Actor%20Object%20Reference) (HitLocation :type Vector) (FHitComponent :type Primitive%20Component%20Object%20Reference) (BoneName :type Name) (Momentum :type Vector))
  ;[1 1 1] Bullet Impact FX
  (Ability|ExecuteGameplayCueWithParamsOnOwner (Variables|Default|GetGCNDamageTaken) (Ability|GameplayCue|MakeGameplayCueParameters :NormalizedMagnitude 0.0 :RawMagnitude 0.0 :Location HitLocation :Normal (* Momentum -1.0) :GameplayEffectLevel 1 :AbilityLevel 1 :bReplicateLocationWhenUsingMinimalRepProxy false)))
