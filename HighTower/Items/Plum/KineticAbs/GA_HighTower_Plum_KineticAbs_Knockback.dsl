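; Gameplay ability for Black Panther's Kinetic Shockwave knockback, a child of a knockback parent ability that extends its launch event.
; On launch it runs the parent launch with the velocity, then applies a quest GameplayEffect to the knockback target's ability system.
; That effect tags knocked-back enemies so later damage to them can count toward a Season 14 Week 6 quest.
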(event EventDoLaunch
  (params (Vel :type Vector :default "0, 0, 0"))
  (|CallParentFunction Vel)
  (GameplayEffects|ApplyGameplayEffectToSelf :self (Ability|GetAbilitySystemComponent (Variables|KnockbackParent|GetTarget)) :GameplayEffectClass "/HighTower/Items/Plum/KineticAbs/GE_HighTower_Plum_Quest.GE_HighTower_Plum_Quest_C" :Level 0.0) ; Season 14 - Week 6 - Quest 7\nDeal Damage after knocking an opponent back with Black Panther's Kinetic Shockwave\nApply tag to knocked-back enemies so damage to them can be tracked
)
